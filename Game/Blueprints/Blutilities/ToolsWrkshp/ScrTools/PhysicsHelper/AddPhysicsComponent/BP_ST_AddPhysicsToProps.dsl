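; Editor scriptable tool that adds or removes FortPhysicsComponent on selected BuildingProp actors, turning them into physics or non-physics props.
; Setup binds four property-set events to add, remove and two selection-filter functions; shutdown unbinds them.
; Each action calls UpdateSelectedActors to sort the selection into valid, physics and non-prop arrays, then works inside a transaction.
; Components are edited through SubobjectDataSubsystem, with bCustomSimulatePhysics enabled on add and results shown as tool messages.
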(event ScriptableTool|Events|EventOnScriptSetup
  (bind _settings (Variables|Default|GetSettings))
  (bind _returnvalue (ScriptableTool|PropertySets|AddPropertySetOfType "/Game/Blueprints/Blutilities/ToolsWrkshp/ScrTools/PhysicsHelper/AddPhysicsComponent/BP_STPS_AddPhysicsToProps.BP_STPS_AddPhysicsToProps_C")
    (:Success
      (bind _asbp_stps_add_physics_to_props (Utilities|Casting|CastToBP_STPS_AddPhysicsToProps _returnvalue))
      (Variables|Default|SetSettings _asbp_stps_add_physics_to_props)
      (Default|BindEventtoEventonAddComponentClicked _settings (EventDispatchers|CreateEvent :!event AddComponents))
      (Default|BindEventtoEventonRemoveComponentClicked _settings (EventDispatchers|CreateEvent :!event RemoveComponents))
      (Default|BindEventtoEventonSelectPhysicsProps _settings (EventDispatchers|CreateEvent :!event SelectPhysProps))
      (Default|BindEventtoEventonSelectInvalidProps _settings (EventDispatchers|CreateEvent :!event SelectNonPhysProps)))))

(event ScriptableTool|Events|EventOnScriptShutdown
  (params (ShutdownType :type EToolShutdownType%20Enum :default "Completed"))
  (bind _settings (Variables|Default|GetSettings))
  (Utilities|IsValid (Variables|Default|GetSettings)
    (:Is%20Valid
      (Default|UnbindEventfromEventonAddComponentClicked _settings (EventDispatchers|CreateEvent :!event AddComponents))
      (Default|UnbindEventfromEventonRemoveComponentClicked _settings (EventDispatchers|CreateEvent :!event RemoveComponents))
      (Default|UnbindEventfromEventonSelectPhysicsProps _settings (EventDispatchers|CreateEvent :!event SelectPhysProps))
      (Default|UnbindEventfromEventonSelectInvalidProps _settings (EventDispatchers|CreateEvent :!event SelectNonPhysProps)))))

(fn AddComponents
  (bind _returnvalue (EngineSubsystems|GetSubobjectDataSubsystem))
  (Default|UpdateSelectedActors)
  (Transactions|BeginTransaction (LocalVariables|GetUndoContext) (LocalVariables|GetUndoDescription))
  (for _array_element (Variables|Default|GetValidActors)
    ;[1 1 1] Iterate through each valid object in the selection and add a FortPhysics component. This will make it into a "Physics Prop"
    (bind _outarray (SubobjectDataSubsystem|GatherSubobjectDataForInstance _returnvalue _array_element))
    (bind (_failreason _returnvalue_2) (SubobjectDataSubsystem|AddNewSubobject _returnvalue (Utilities|Struct|MakeAddNewSubobjectParams (Utilities|Array|Get(aref) _outarray) "/Script/FortniteGame.FortPhysicsComponent")))
    (bind _asfortphysics (Utilities|Casting|CastToFortPhysicsComponent (SubobjectData|GetAssociatedObject (SubobjectData|GetData _returnvalue_2))))
    (Utilities|SetEditorProperty _asfortphysics "bCustomSimulatePhysics" true))
  (Transactions|EndTransaction)
  ;[1 1 1] Handle Logs for the tool.
  (ScriptableTool|Messaging|ClearUserMessages true)
  (ScriptableTool|Messaging|DisplayUserWarningMessage (Utilities|Text|ToText(String) (Utilities|String|Append "Finished adding FortPhysics Components.

Number of Physics Props created: " (Utilities|String|ToString(Integer) (Utilities|Array|Length (Variables|Default|GetValidActors))))))
  (Default|ToolLogs)
  (return))

(fn RemoveComponents
  (bind _returnvalue (EngineSubsystems|GetSubobjectDataSubsystem))
  (Default|UpdateSelectedActors)
  (Transactions|BeginTransaction (LocalVariables|GetContext) (LocalVariables|GetDescription))
  (for _array_element (Variables|Default|GetPhysicsProps)
    ;[1 1 1] Iterate through the Physics props we have selected, then find the FortPhysics Component Suboject, and remove it from its handle.
    (bind _outarray (SubobjectDataSubsystem|GatherSubobjectDataForInstance _returnvalue _array_element))
    (SubobjectDataSubsystem|DeleteSubobjectfromInstance _returnvalue (Utilities|Array|Get(aref) _outarray) (SubobjectDataSubsystem|FindHandleforObject _returnvalue (Utilities|Array|Get(aref) _outarray) (Actor|GetComponentbyClass _array_element "/Script/FortniteGame.FortPhysicsComponent") 0)))
  (Transactions|EndTransaction)
  (ScriptableTool|Messaging|ClearUserMessages true)
  (ScriptableTool|Messaging|DisplayUserWarningMessage (Utilities|Text|ToText(String) (Utilities|String|Append "Finished removing FortPhysics Components.

Number of props converted to Non-Physics Props:  " (Utilities|String|ToString(Integer) (Utilities|Array|Length (Variables|Default|GetPhysicsProps))))))
  (return))

(fn SelectNonPhysProps
  (Default|UpdateSelectedActors)
  (Transactions|BeginTransaction (LocalVariables|GetContext) (LocalVariables|GetDescription))
  ;[1 1 1] Set the selection to Valid Actors since those are the props that are currently not Physics Props, but are valid targets for the component.
  (EditorScripting|LevelUtility|SetSelectedLevelActors (EditorSubsystems|GetEditorActorSubsystem) (Variables|Default|GetValidActors))
  (Transactions|EndTransaction)
  (return))

(fn SelectPhysProps
  (Default|UpdateSelectedActors)
  (Transactions|BeginTransaction (LocalVariables|GetContext) (LocalVariables|GetDescription))
  ;[1 1 1] Since Physics Props array is the list we are looking for, we set that to be the new selection
  (EditorScripting|LevelUtility|SetSelectedLevelActors (EditorSubsystems|GetEditorActorSubsystem) (Variables|Default|GetPhysicsProps))
  (Transactions|EndTransaction)
  (return))

(fn ToolLogs
  (if (Utilities|Array|IsNotEmpty (Variables|Default|GetNonPropActors))
    ;[1 1 1] Log in-tool: Print the number of actors that we couldn't add the FortPhysics component to because they are not a prop.
    (ScriptableTool|Messaging|DisplayUserWarningMessage (Utilities|Text|ToText(String) (Utilities|String|Append "Unable to add FortPhysics component to " (Utilities|String|ToString(Integer) (Utilities|Array|Length (Variables|Default|GetNonPropActors))) " ctors. Reason: Actor is not a valid prop.")))
    (else
      _))
  ;[1 1 1] Log in-tool: Print the number of actors that we couldn't add the FortPhysics component to because they already have a FortPhysics Component.
  (if (Utilities|Array|IsNotEmpty (Variables|Default|GetPhysicsProps))
    (ScriptableTool|Messaging|DisplayUserWarningMessage (Utilities|Text|ToText(String) (Utilities|String|Append "Unable to add FortPhysics component to " (Utilities|String|ToString(Integer) (Utilities|Array|Length (Variables|Default|GetPhysicsProps))) " actors. Reason: Actor already has a FortPhysics component.")))
    (else
      _))
  (return))

(fn UpdateSelectedActors
  (bind _returnvalue (EditorScripting|LevelUtility|GetSelectedLevelActors (EditorSubsystems|GetEditorActorSubsystem)))
  (if (Utilities|Array|IsNotEmpty _returnvalue)
    (Utilities|Array|Clear (Variables|Default|GetNonPropActors))
    (Utilities|Array|Clear (Variables|Default|GetPhysicsProps))
    (Utilities|Array|Clear (Variables|Default|GetValidActors))
    (for _array_element _returnvalue
      (bind _asbuilding_prop (Utilities|Casting|CastToBuildingProp _array_element)
        (:then
          (if (not (Utilities|IsValid (Actor|GetComponentbyClass _asbuilding_prop "/Script/FortniteGame.FortPhysicsComponent") :!class /Script/Engine.KismetSystemLibrary))
            (Utilities|Array|Add (Variables|Default|GetValidActors) _asbuilding_prop)
            (else
              (Utilities|Array|Add (Variables|Default|GetPhysicsProps) _asbuilding_prop))))
        (:CastFailed
          (Utilities|Array|Add (Variables|Default|GetNonPropActors) _array_element))))
    (else
      _))
  (return _returnvalue))
