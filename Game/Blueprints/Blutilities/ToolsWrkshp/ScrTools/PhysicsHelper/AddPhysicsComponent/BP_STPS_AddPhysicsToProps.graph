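# Editor utility Blueprint that relays add physics, remove physics and selection commands for a props physics tool.
# Add Physics, Remove Physics, Select Non Physics and Select Physics each broadcast a matching event dispatcher with no parameters.
# The event graph is empty and no physics or selection logic lives here, so the work depends on whatever binds to those dispatchers.
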EditorUtilityBlueprint /Game/Blueprints/Blutilities/ToolsWrkshp/ScrTools/PhysicsHelper/AddPhysicsComponent/BP_STPS_AddPhysicsToProps
graphs [
 Add Physics {nodes [
   Add Physics() {is=title K2Node_FunctionEntry}
   > Call Event on Add Component Clicked!(execute:Exec, self:BP STPS Add Physics To Props Object Reference) -> (then:Exec) {is=title K2Node_CallDelegate}
  ]}
 EventGraph {links [], nodes []}
 EventOnAddComponentClicked {nodes [EventOnAddComponentClicked() {is=title K2Node_FunctionEntry}]}
 EventOnRemoveComponentClicked {nodes [EventOnRemoveComponentClicked() {is=title K2Node_FunctionEntry}]}
 EventOnSelectInvalidProps {nodes [EventOnSelectInvalidProps() {is=title K2Node_FunctionEntry}]}
 EventOnSelectPhysicsProps {nodes [EventOnSelectPhysicsProps() {is=title K2Node_FunctionEntry}]}
 Remove Physics {nodes [
   Remove Physics() {is=title K2Node_FunctionEntry}
   > Call Event on Remove Component Clicked!(execute:Exec, self:BP STPS Add Physics To Props Object Reference) -> (then:Exec) {is=title K2Node_CallDelegate}
  ]}
 Select Non Physics {nodes [
   Select Non Physics() {is=title K2Node_FunctionEntry}
   > Call Event on Select Invalid Props!(execute:Exec, self:BP STPS Add Physics To Props Object Reference) -> (then:Exec) {is=title K2Node_CallDelegate}
  ]}
 Select Physics {nodes [
   Select Physics() {is=title K2Node_FunctionEntry}
   > Call Event on Select Physics Props!(execute:Exec, self:BP STPS Add Physics To Props Object Reference) -> (then:Exec) {is=title K2Node_CallDelegate}
  ]}
]
variables [
 EventOnAddComponentClicked {default=()}
 EventOnRemoveComponentClicked {default=()}
 EventOnSelectInvalidProps {default=()}
 EventOnSelectPhysicsProps {default=()}
]
