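# Editor scriptable tool that adds or removes FortPhysicsComponents on selected props, or narrows selection to physics or non-physics props.
# UpdateSelectedActors sorts selection into valid, already-physics and invalid lists, accepting BuildingSMActors except creative device props.
# Transacted actions edit instance components through SubobjectDataSubsystem, enabling bCustomSimulatePhysics on additions; add and remove report counts.
# Setup binds settings property-set events and creates a header widget; tick builds action buttons once.
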EditorUtilityBlueprint /Game/Blueprints/Blutilities/ToolsWrkshp/ScrTools/PhysicsHelper/AddPhysicsComponent/BP_ST_AddPhysicsToProps2
graphs [
 AddComponents {nodes [
   AddComponents() {is=title K2Node_FunctionEntry}
   > UpdateSelectedActors!(execute, self) -> (then, CurrentlySelectedActors:Array of Actor Object References)
   > BeginTransaction(Context=Get UndoContext() -> (UndoContext:String), Description=Get UndoDescription() -> (UndoDescription:Text))
   > n1 = For Each Loop(Array:Array of Actor Object References=Get ValidActors() -> (ValidActors:Array of Actor Object References)) -> (Array Element:Actor Object Reference)
   EndTransaction(execute=@n1.Completed)
   > ClearUserMessages!(execute, self:Scriptable Interactive Tool Object Reference, bHelpMessage:bool=true, bWarningMessage:bool=true) -> (then)
   n2 = "To String (Integer)"(InInt=Length(TargetArray=Get ValidActors()))
   > DisplayUserWarningMessage(Message="To Text (String)"(InString=Append(A="Finished adding FortPhysics Components.\r\n\r\nNumber of Physics Props created: ", B=@n2)))
   > ToolLogs()
   > Return Node()
   n3 = Get SubobjectDataSubsystem()
   n4 = Gather Subobject Data For Instance(execute=@n1.LoopBody, self=@n3, Context=@n1)
   > n5 = AddNewSubobject!(
    execute
    self:Subobject Data Subsystem Object Reference=@n3
    Params:"Add New Subobject Params Structure (by ref)"=Make Add New Subobject Params!(
     ParentHandle:Subobject Data Handle Structure="Get (a ref)"(Array:Array of Subobject Data Handle Structures=@n4, Dimension 1=0) -> (Output:"Subobject Data Handle Structure (by ref)")
     NewClass:Object Class Reference=/Script/FortniteGame.FortPhysicsComponent
     BlueprintContext:Blueprint Object Reference
     bSkipMarkBlueprintModified:bool=False
     bConformTransformToParent:bool=True
    ) -> (AddNewSubobjectParams:Add New Subobject Params Structure) {is=title K2Node_MakeStruct}
   ) -> (then, FailReason:Text, ReturnValue:Subobject Data Handle Structure)
   > n6 = Cast To FortPhysicsComponent!(
    execute
    Object=GetAssociatedObject!(Data:"Subobject Data Structure (by ref)"=GetData(DataHandle=@n5.ReturnValue)) -> (ReturnValue:Object Reference)
   ) -> (then, CastFailed, AsFortPhysics:Fort Physics Object Reference) {is=title K2Node_DynamicCast}
   > SetEditorProperty(Object=@n6, PropertyName=bCustomSimulatePhysics, PropertyValue=MakeLiteralBool(Value=true))
  ]}
 EventGraph {nodes [
   Event OnScriptSetup()
   > n1 = Sequence!(execute) -> (then_0, then_1, then_2:Exec)
   > n2 = AddPropertySetOfType(
    PropertySetType=&/BP_STPS_AddPhysicsToProps.BP_STPS_AddPhysicsToProps_C
    Identifier=Settings
   )
   > n3 = Cast To BP_STPS_AddPhysicsToProps!(execute, Object=@n2) -> (then, CastFailed, AsBP STPS Add Physics to Props:BP STPS Add Physics To Props Object Reference) {is=title K2Node_DynamicCast}
   > Set Settings(Settings:BP STPS Add Physics To Props Object Reference=@n3) -> (Output_Get:BP STPS Add Physics To Props Object Reference)
   n4 = Create Event()
   n5 = Get Settings() -> (Settings:BP STPS Add Physics To Props Object Reference)
   > Bind Event to Event on Add Component Clicked(self:BP STPS Add Physics To Props Object Reference=@n5, Delegate=@n4) {is=title K2Node_AddDelegate}
   n6 = Create Event()
   > Bind Event to Event on Remove Component Clicked(self:BP STPS Add Physics To Props Object Reference=@n5, Delegate=@n6) {is=title K2Node_AddDelegate}
   n7 = Create Event()
   > Bind Event to Event on Select Physics Props(self:BP STPS Add Physics To Props Object Reference=@n5, Delegate=@n7) {is=title K2Node_AddDelegate}
   n8 = Create Event()
   > Bind Event to Event on Select Invalid Props(self:BP STPS Add Physics To Props Object Reference=@n5, Delegate=@n8) {is=title K2Node_AddDelegate}
   n9 = Construct Object from Class(
    execute=@n1.then_1
    Class=&BP_UW_SharedHeaderWidget_C
   )
   > n10 = Set Header(Header:BP UW Shared Header Widget Object Reference=@n9) -> (Output_Get:BP UW Shared Header Widget Object Reference)
   > SetToolPanelToolHeaderWidget(Widget=@n10)
   Event OnScriptShutdown()
   > Is Valid(InputObject=Get Settings())
   > Unbind Event from Event on Add Component Clicked(self:BP STPS Add Physics To Props Object Reference=@n5, Delegate=@n4) {is=title K2Node_RemoveDelegate}
   > Unbind Event from Event on Remove Component Clicked(self:BP STPS Add Physics To Props Object Reference=@n5, Delegate=@n6) {is=title K2Node_RemoveDelegate}
   > Unbind Event from Event on Select Physics Props(self:BP STPS Add Physics To Props Object Reference=@n5, Delegate=@n7) {is=title K2Node_RemoveDelegate}
   > Unbind Event from Event on Select Invalid Props(self:BP STPS Add Physics To Props Object Reference=@n5, Delegate=@n8) {is=title K2Node_RemoveDelegate}
   Event OnScriptTick()
   > n11 = Branch(Condition=Get FirstInitRun())
   Set FirstInitRun(execute=@n11.else)
   > n12 = AddExpandableBox(self=Get Header() -> (Header:BP UW Shared Header Widget Object Reference))
   > n13 = Set Actions Box(Actions Box=@n12)
   > SetUpBox(self=@n13, Name=Get Actions Box Label(self=Get Header()))
   > n14 = Sequence!(execute) -> (then_0, then_1, then_2, then_3:Exec)
   > n15 = Construct Object from Class(Class=/Game/Blueprints/Blutilities/ToolsWrkshp/ScrTools/PhysicsHelper/AddPhysicsComponent/SupportFiles/BP_UW_AddPhysicsButton.BP_UW_AddPhysicsButton_C) -> (ReturnValue:BP UW Add Physics Button Object Reference)
   > n16 = Set AddPhysicsButton(AddPhysicsButton:BP UW Add Physics Button Object Reference=@n15) -> (Output_Get:BP UW Add Physics Button Object Reference)
   > AddChildToVerticalBox(self=Get DynamicEntries(self=Get Actions Box()), Content=@n16)
   > Bind Event to Button Clicked(
    self:BP UW Add Physics Button Object Reference=Get AddPhysicsButton() -> (AddPhysicsButton:BP UW Add Physics Button Object Reference)
    Delegate=Create Event()
   ) {is=title K2Node_AddDelegate}
   n17 = Construct Object from Class(
    execute=@n14.then_1
    Class=/Game/Blueprints/Blutilities/ToolsWrkshp/ScrTools/PhysicsHelper/AddPhysicsComponent/SupportFiles/BP_UW_RemovePhysicsButton.BP_UW_RemovePhysicsButton_C
   ) -> (ReturnValue:BP UW Remove Physics Button Object Reference)
   > n18 = Set Remove Physics Button(Remove Physics Button:BP UW Remove Physics Button Object Reference=@n17) -> (Output_Get:BP UW Remove Physics Button Object Reference)
   > AddChildToVerticalBox(self=Get DynamicEntries(self=Get Actions Box()), Content=@n18)
   > Bind Event to Button Clicked(
    self:BP UW Remove Physics Button Object Reference=Get Remove Physics Button() -> (Remove Physics Button:BP UW Remove Physics Button Object Reference)
    Delegate=Create Event()
   )
   n19 = Construct Object from Class(
    execute=@n14.then_2
    Class=/Game/Blueprints/Blutilities/ToolsWrkshp/ScrTools/PhysicsHelper/AddPhysicsComponent/SupportFiles/BP_UW_SelectPhysicsButton.BP_UW_SelectPhysicsButton_C
   ) -> (ReturnValue:BP UW Select Physics Button Object Reference)
   > n20 = Set Select Physics Button(Select Physics Button:BP UW Select Physics Button Object Reference=@n19) -> (Output_Get:BP UW Select Physics Button Object Reference)
   > AddChildToVerticalBox(self=Get DynamicEntries(self=Get Actions Box()), Content=@n20)
   > Bind Event to Button Clicked(
    self:BP UW Select Physics Button Object Reference=Get Select Physics Button() -> (Select Physics Button:BP UW Select Physics Button Object Reference)
    Delegate=Create Event()
   )
   n21 = Construct Object from Class(
    execute=@n14.then_3
    Class=/Game/Blueprints/Blutilities/ToolsWrkshp/ScrTools/PhysicsHelper/AddPhysicsComponent/SupportFiles/BP_UW_SelectNonPhysicsButton.BP_UW_SelectNonPhysicsButton_C
   ) -> (ReturnValue:BP UW Select Non Physics Button Object Reference)
   > n22 = Set Select Non Physics Button(Select Non Physics Button:BP UW Select Non Physics Button Object Reference=@n21) -> (Output_Get:BP UW Select Non Physics Button Object Reference)
   > AddChildToVerticalBox(self=Get DynamicEntries(self=Get Actions Box()), Content=@n22)
   > Bind Event to Button Clicked(
    self:BP UW Select Non Physics Button Object Reference=Get Select Non Physics Button() -> (Select Non Physics Button:BP UW Select Non Physics Button Object Reference)
    Delegate=Create Event()
   )
  ]}
 RemoveComponents {nodes [
   RemoveComponents() {is=title K2Node_FunctionEntry}
   > UpdateSelectedActors()
   > BeginTransaction(Context=Get Context() -> (Context:String), Description=Get Description() -> (Description:Text))
   > n1 = For Each Loop(Array=Get PhysicsProps() -> (PhysicsProps:Array of Actor Object References))
   EndTransaction(execute=@n1.Completed)
   > ClearUserMessages(bHelpMessage=true, bWarningMessage=true)
   n2 = "To String (Integer)"(InInt=Length(TargetArray=Get PhysicsProps()))
   > DisplayUserWarningMessage(Message="To Text (String)"(InString=Append(A="Finished removing FortPhysics Components.\r\n\r\nNumber of props converted to Non-Physics Props:  ", B=@n2)))
   > Return Node()
   n3 = Get SubobjectDataSubsystem()
   n4 = Gather Subobject Data For Instance(execute=@n1.LoopBody, self=@n3, Context=@n1)
   n5 = "Get (a ref)"(Array=@n4, Dimension 1=0)
   > Delete Subobject from Instance!(
    execute
    self:Subobject Data Subsystem Object Reference=@n3
    ContextHandle:"Subobject Data Handle Structure (by ref)"=@n5
    SubobjectToDelete:"Subobject Data Handle Structure (by ref)"=FindHandleForObject!(
     self:Subobject Data Subsystem Object Reference=@n3
     Context:"Subobject Data Handle Structure (by ref)"=@n5
     ObjectToFind:Object Reference=GetComponentByClass!(self:Actor Object Reference=@n1, ComponentClass:Actor Component Class Reference=&FortPhysicsComponent) -> (ReturnValue:Fort Physics Object Reference)
     BPContext:Blueprint Object Reference
    ) -> (ReturnValue:Subobject Data Handle Structure)
   ) -> (then, ReturnValue:int=0)
  ]}
 SelectNonPhysProps {nodes [
   SelectNonPhysProps() {is=title K2Node_FunctionEntry}
   > UpdateSelectedActors()
   > BeginTransaction(Context=Get Context(), Description=Get Description())
   > SetSelectedLevelActors(self=Get EditorActorSubsystem(), ActorsToSelect=Get ValidActors())
   > EndTransaction()
   > Return Node()
  ]}
 SelectPhysProps {nodes [
   SelectPhysProps() {is=title K2Node_FunctionEntry}
   > UpdateSelectedActors()
   > BeginTransaction(Context=Get Context(), Description=Get Description())
   > SetSelectedLevelActors(self=Get EditorActorSubsystem(), ActorsToSelect=Get PhysicsProps())
   > EndTransaction()
   > Return Node()
  ]}
 ToolLogs {nodes [
   ToolLogs() {is=title K2Node_FunctionEntry}
   > n1 = Branch(Condition=Is Not Empty!(TargetArray:Array of Actor Object References=Get InvalidActors() -> (InvalidActors:Array of Actor Object References)) -> (ReturnValue:bool))
   n2 = "To String (Integer)"(InInt=Length(TargetArray=Get InvalidActors()))
   > n3 = DisplayUserWarningMessage(Message="To Text (String)"(InString=Append!(A="Unable to add FortPhysics component to ", B=@n2, C:String=" ctors. Reason: Actor is not a valid prop.") -> (ReturnValue)))
   n4 = Branch(execute=[@n3.then, @n1.else], Condition=Is Not Empty(TargetArray=Get PhysicsProps()))
   n5 = "To String (Integer)"(InInt=Length(TargetArray=Get PhysicsProps()))
   > n6 = DisplayUserWarningMessage(Message="To Text (String)"(InString=Append(A="Unable to add FortPhysics component to ", B=@n5, C=" actors. Reason: Actor already has a FortPhysics component.")))
   Return Node(execute=[@n6.then, @n4.else])
  ]}
 UpdateSelectedActors {nodes [
   UpdateSelectedActors() {is=title K2Node_FunctionEntry}
   > n1 = GetSelectedLevelActors(self=Get EditorActorSubsystem())
   > n2 = Branch(Condition=Is Not Empty(TargetArray=@n1))
   > Clear(TargetArray:Array of Actor Object References=Get InvalidActors())
   > Clear(TargetArray=Get PhysicsProps())
   > Clear(TargetArray=Get ValidActors())
   > n3 = For Each Loop(Array=@n1)
   Return Node!(execute=[@n3.Completed, @n2.else], CurrentlySelectedActors:Array of Actor Object References=@n1)
   > n4 = Cast To BuildingSMActor!(execute, Object=@n3) -> (then, CastFailed, AsBuilding Static Mesh:Building Static Mesh Object Reference) {is=title K2Node_DynamicCast}
   > n5 = Cast To FortCreativeDeviceProp!(execute, Object=@n3) -> (then, CastFailed, AsFort Creative Device Prop:Fort Creative Device Prop Object Reference) {is=title K2Node_DynamicCast}
   Add(
    execute=[@n5.then, @n4.CastFailed]
    TargetArray:Array of Actor Object References=Get InvalidActors()
    NewItem:"Actor Object Reference (by ref)"=@n3
   ) {is=title K2Node_CallArrayFunction}
   n6 = GetComponentByClass(self=@n3, ComponentClass=&FortPhysicsComponent)
   n7 = Branch(execute=@n5.CastFailed, Condition=NOT Boolean(A=IsValid(Object=@n6)))
   > Add(TargetArray=Get ValidActors(), NewItem=@n3)
   Add(execute=@n7.else, TargetArray=Get PhysicsProps(), NewItem=@n3)
  ]}
]
variables [
 Actions Box {type=BP_UW_ExpandableBox_C}
 AddPhysicsButton {type=BP_UW_AddPhysicsButton_C, default=null}
 FirstInitRun {type=boolean}
 Header {type=BP_UW_SharedHeaderWidget_C}
 InvalidActors {type="array<Actor>"}
 PhysicsProps {type="array<Actor>"}
 Remove Physics Button {type=BP_UW_RemovePhysicsButton_C, default=null}
 Select Non Physics Button {type=BP_UW_SelectNonPhysicsButton_C, default=null}
 Select Physics Button {type=BP_UW_SelectPhysicsButton_C, default=null}
 Settings {type=BP_STPS_AddPhysicsToProps_C, default=null}
 ValidActors {type="array<Actor>"}
]
